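; Main cluster bomb grenade projectile that lands, explodes after a delay, then releases small grenades.
; On stop it snaps to the hit normal, fires a pre-detonate cue, attaches to tagged actors and wobbles its plunger; water entry also starts the delayed explosion.
; After exploding, the authority spawns children on a looping timer, aimed around the hit normal in a jittered ring, with the first along the normal.
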(event EventBeginPlay
  ;[0.03429 0.3069 0.7899] BeginPlay
  (|CallParentFunction)
  (Game|Components|ProjectileMovement|Interpolation|SetInterpolatedComponent (Variables|Movement|GetProjectileMovementComponent) (Variables|Default|GetBomb_Mesh)))

(event Custom|StopRotation
  ;[0.03429 0.3069 0.7899] Snap rotation to hit normal
  ;;[1 1 1] Set interpolated component to null to ensure the mesh snaps to the final location as expected
  (Game|Components|ProjectileMovement|Interpolation|SetInterpolatedComponent (Variables|Movement|GetProjectileMovementComponent) 0)
  (Components|Activation|Deactivate (Variables|Default|GetRotatingMovement))
  (Transformation|SetActorLocationAndRotation (Variables|Default|GetHitLocation) (Math|Conversions|RotationFromXVector (Math|Vector|NegateVector (Variables|Default|GetHitNormal)))))

(event EventOnStop
  (params (Hit :type Hit%20Result%20Structure%20%28by%20ref%29))
  (bind (_bblockinghit _binitialoverlap _time _distance _location _impactpoint _normal _impactnormal _physmat _hitactor) (Collision|BreakHitResult Hit))
  ;[0.03429 0.3069 0.7899] Projectile stopped moving
  (Variables|Default|SetHitNormal _impactnormal)
  (Variables|Default|SetHitLocation _location)
  (GameplayCue|ExecuteGameplayCueOnActor(Burst) self (Variables|Default|GetGC_PreDetonate) (GameplayCue|MakeGameplayCueParametersfromHitResult Hit))
  (CallFunction|StopRotation)
  (Variables|Default|SetbHitMovingPlatform (Utilities|IsValid (Actor|GetComponentbyClass _hitactor "/Script/FortniteGame.FortMovingPlatformComponent") :!class /Script/Engine.KismetSystemLibrary))
  (bind _returnvalue (GameplayTags|HasAnyMatchingGameplayTags(Message) _hitactor (Variables|Default|GetAttachableTags)))
  (if _returnvalue
    (Transformation|AttachActorToComponent (Variables|/Script/Engine.Actor|GetRootComponent _hitactor) "None" "KeepWorld" "KeepWorld" "KeepWorld" false)
    (else
      _))
  (Utilities|FlowControl|Sequence
    (:then_0
      (Utilities|FlowControl|Delay (ScalableFloat|GetValueatLevel (Variables|Default|GetHF_ExplosionDelay)))
      (Delivery|DoExplosion))
    (:then_1
      (Utilities|EffectsQualityCheck 0
        (:Cinematic
          _)
        (:Epic
          _))
      (bind (_direction _wobblecurve) (Timeline|Timeline PlayFromStart)
        (:Update
          (Transformation|SetRelativeRotation (Variables|Default|GetArrow_Plunger) (Math|Rotator|MakeRotator 0.0 (* _wobblecurve 3.0) (* _wobblecurve 3.0)) false false))))))

(event EventOnExploded
  (params (HitActors :type Array%20of%20Actor%20Object%20References) (HitResults :type Array%20of%20Hit%20Result%20Structures))
  ;[0.03429 0.3069 0.7899] Explode and spawn small grenades
  (|CallParentFunction HitActors HitResults)
  (Collision|SetActorEnableCollisionwithChildren 0)
  (Networking|SwitchHasAuthority
    (:Authority
      (bind _returnvalue (Utilities|Time|SetTimerbyEvent (delegate Custom|SpawnGrenadeTimed) (ScalableFloat|GetValueatLevel (Variables|Default|GetHF_GrenadeInterval)) true))
      (Variables|Default|SetGrenadeSpawnTimer _returnvalue)
      (bind _output_get (Variables|Default|SetGrenadesSpawned 0))
      (CallFunction|SpawnGrenadeSmall _output_get))))

(event Custom|SpawnGrenadeTimed
  (bind _grenadesspawned (Variables|Default|GetGrenadesSpawned))
  ;[0.03429 0.3069 0.7899] Explode and spawn small grenades
  (CallFunction|SpawnGrenadeSmall _grenadesspawned)
  (if (>= _grenadesspawned (ScalableFloat|GetValueasInteger (Variables|Default|GetHF_GrenadesToSpawn)))
    (Utilities|Time|ClearandInvalidateTimerbyHandle (Variables|Default|GetGrenadeSpawnTimer))
    (Actor|DestroyActor)))

(event OnEnterWaterEvent(WaterInteractionComponent)
  (params (WaterBody :type Fort%20Water%20Body%20Actor%20Object%20Reference) (WaterInteractionComponent :type Fort%20Water%20Interaction%20Component%20Object%20Reference) (bIsFirstBody :type Boolean))
  (bind (_waterplanelocation _waterplanenormal _watersurfaceposition _waterdepth _waterbodyidx) (FX|GetLastWaterSurfaceInfo WaterInteractionComponent :!class /Script/FortniteGame.FortWaterInteractionComponent))
  ;[0.03429 0.3069 0.7899] Handle Entering Water, reused from OG Clinger logic
  (Utilities|FlowControl|DoOnce false
    (:Completed
      (Variables|Default|SetEnteredWater? false)
      (Variables|Default|SetHitLocation (Transformation|GetActorLocation))
      (Variables|Default|SetHitNormal "0, 0,1.000000")
      (bind _returnvalue (Effects|SpawnEmitteratLocation "/Game/Weapons/FORT_Generic_Effects/Impacts/Water/P_Pistol_Impact_Water_Apollo.P_Pistol_Impact_Water_Apollo" _watersurfaceposition (Math|Rotator|MakeRotfromX _waterplanenormal)))
      (Audio|PlaySoundatLocation "/ClusterBombGameplay/Audio/ClusterBomb_ThrowLand_Water.ClusterBomb_ThrowLand_Water" _watersurfaceposition)
      (Effects|Components|ParticleSystem|SetFloatParameter _returnvalue "WaterDepth" _waterdepth)
      (Effects|Components|ParticleSystem|SetFloatParameter _returnvalue "WaterBodyIndex" (Math|Conversions|ToFloat(Integer) _waterbodyidx))
      (Audio|PlaySoundatLocation "/Game/Sounds/Fort_GamePlay_Sounds/Explosives/Grenade_Projectile_Bounce_Water_Splash_NoLoop_Cue.Grenade_Projectile_Bounce_Water_Splash_NoLoop_Cue" (Transformation|GetActorLocation))
      (Variables|/Script/Engine.ProjectileMovementComponent|SetbRotationFollowsVelocity false (Variables|Movement|GetProjectileMovementComponent))
      (CallFunction|StopRotation)
      (Utilities|FlowControl|Delay (ScalableFloat|GetValueatLevel (Variables|Default|GetHF_ExplosionDelay)))
      (Delivery|DoExplosion))))

(fn CalculateSmallGrenadeSpawnDirection
  (params (Index :type Integer) (HitNormal :type Vector))
  (bind _index (Variables|GetIndex))
  (bind _returnvalue (== _index 0))
  (bind _grenadesmallspawnverticalangle (Variables|SmallGrenade|GetGrenadeSmallSpawnVerticalAngle))
  (bind _returnvalue_2 (Math|Float|SelectFloat 9e+01 _grenadesmallspawnverticalangle _returnvalue))
  (bind _hitnormal (Variables|Default|GetHitNormal))
  (bind _crossproduct (LocalVariables|GetCrossProduct))
  (bind _returnvalue_3 (Math|Vector|CrossProduct _crossproduct _hitnormal))
  (bind _crossproduct_2 (LocalVariables|GetCrossProduct))
  (bind _returnvalue_4 (Math|Vector|RotateVectorAroundAxis _crossproduct_2 _returnvalue_2 _returnvalue_3))
  (bind _hf_grenadestospawn (Variables|Default|GetHF_GrenadesToSpawn))
  (bind _returnvalue_5 (ScalableFloat|GetValueasInteger _hf_grenadestospawn))
  (bind _grenadesmallhorizontalanglevariance (Variables|SmallGrenade|GetGrenadeSmallHorizontalAngleVariance))
  (bind _grenadesmallhorizontalanglevariance_2 (Variables|SmallGrenade|GetGrenadeSmallHorizontalAngleVariance))
  (bind _returnvalue_6 (* _grenadesmallhorizontalanglevariance_2 -1.0))
  (bind _returnvalue_7 (Math|Random|RandomFloatinRange _returnvalue_6 _grenadesmallhorizontalanglevariance))
  (bind _returnvalue_8 (- _returnvalue_5 1))
  (bind _returnvalue_9 (/ 360 _returnvalue_8))
  (bind _index_2 (Variables|GetIndex))
  (bind _returnvalue_10 (* _index_2 _returnvalue_9))
  (bind _returnvalue_11 (Math|Conversions|ToFloat(Integer) _returnvalue_10))
  (bind _hitnormal_2 (Variables|Default|GetHitNormal))
  (bind _returnvalue_12 (+ _returnvalue_11 _returnvalue_7))
  (bind _returnvalue_13 (Math|Vector|RotateVectorAroundAxis _returnvalue_4 _returnvalue_12 _hitnormal_2))
  ;[1 1 1] Cross product to get vector parallel with surface
  (bind _output_get (LocalVariables|SetCrossProduct (Math|Vector|CrossProduct HitNormal (select (> (Math|Float|Absolute(Float) (.x HitNormal)) 0.9) (LocalVariables|GetYAxis) (LocalVariables|GetXAxis)))))
  (Debug|DrawDebugDirectionalArrow _output_get "(R=1.000000,G=0.000000,B=0.000000,A=1.000000)")
  ;[1 1 1] Rotate around HitNormal based on Index and horizontal angle variance
  (Debug|DrawDebugDirectionalArrow _returnvalue_13 "(R=0.933333,G=0.000000,B=1.000000,A=1.000000)")
  (return (Math|Conversions|RotationFromXVector _returnvalue_13)))

(fn DrawDebugDirectionalArrow
  (params (LineEnd :type Vector) (LineColor :type Linear%20Color%20Structure))
  (if (and (Variables|Debug|GetbDrawDebugDirectionalArrows) (Editor|InEditororPIE))
    (Rendering|Debug|DrawDebugArrow (Transformation|GetActorLocation) (+ (Transformation|GetActorLocation) (* (Variables|GetLineEnd) 5e+01)) 1e+01 (Variables|GetLineColor) 5.0 2.0)))

(fn SpawnGrenadeSmall
  (params (Index :type Integer))
  (LocalVariables|SetExplosionDelay (* Index (ScalableFloat|GetValueatLevel (Variables|Default|GetHF_GrenadeInterval))))
  (bind _returnvalue (Ability|MakeGameplayEffectContainerSpecfromAbilityComponent (Variables|Default|GetGrenadeEffectContainer) (Ability|GetFortAbilitySystemComponent (Game|GetInstigator))))
  (Debug|DrawDebugDirectionalArrow (Variables|Default|GetHitNormal) "(R=1.000000,G=0.000000,B=0.000000,A=1.000000)")
  (bind _spawnrotation (CallFunction|CalculateSmallGrenadeSpawnDirection (Variables|GetIndex) (Variables|Default|GetHitNormal)))
  (bind _returnvalue_2 (Projectile|SpawnProjectilewithParams (Variables|Default|GetGrenadeSmallClass) (Game|GetInstigator) (Utilities|Struct|MakeSpawnProjectileParams (Transformation|GetActorLocation) "0.000000,0.000000,0.000000" _spawnrotation "(TargetSelection=(List=,bStopAtFirstSuccess=True,bKeepCheckingListOnIndestructibleHit=False,AbilityTargetBlockedGameplayTags=(GameplayTags=),bUseWeaponRanges=True,bUseMaxYawAngleToTarget=False,MaxYawAngleToTarget=0.000000),TargetGameplayEffectSpecs=,OwnerGameplayEffectSpecs=,ActivationCues=(GameplayTags=),ImpactCues=(GameplayTags=),ImpactNoiseRange=0.000000,FlyByNoiseRange=0.000000,bOverrideChargeMagnitude=False,ChargeMagnitudeOverrideValue=0.000000)" _returnvalue false 0 (ScalableFloat|GetValueatLevel (Variables|SmallGrenade|GetHF_SmallGrenadeSpeed)) (ScalableFloat|GetValueatLevel (Variables|SmallGrenade|GetHF_SmallGrenadeSpeed)) (ScalableFloat|GetValueatLevel (Variables|SmallGrenade|GetHF_SmallGrenadeGravity)))))
  (bind _asb_prj_paprika_cluster_bomb_grenade_small (Utilities|Casting|CastToB_Prj_Paprika_ClusterBomb_GrenadeSmall _returnvalue_2)
    (:then
      (Variables|/ClusterBombGameplay/Gameplay/Actors/B_Prj_Paprika_ClusterBomb_GrenadeSmall.B_Prj_Paprika_ClusterBomb_GrenadeSmall_C|SetbMainGrenadeStuckToMovingPlatform (Variables|Default|GetbHitMovingPlatform) _asb_prj_paprika_cluster_bomb_grenade_small))
    (:CastFailed
      _))
  (Math|Integer|IncrementInt (Variables|Default|GetGrenadesSpawned)
    (:%20%20
      (return))))

(fn ConstructionScript)
