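; Small cluster bomb grenade projectile that detonates on a timer and plays surface-dependent explosion effects.
; BeginPlay fires a spawn gameplay cue and sets the timer from a scalable base delay plus DetonationDelay.
; After the parent's explosion logic, cosmetic-capable machines show a sound indicator and play water or normal FX and audio.
; The normal FX orients its decal by the stop-hit normal, suppressing it if the main grenade stuck to a moving platform.
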(event EventBeginPlay
  ;[0.03429 0.3069 0.7899] Do delayed explosion
  (Utilities|Time|SetTimerbyEvent (delegate Custom|DoExplosionLatent) (+ (ScalableFloat|GetValueatLevel (Variables|Default|GetHF_BaseExplosionDelay)) (Variables|Default|GetDetonationDelay)))
  (GameplayCue|ExecuteGameplayCueOnActor(Burst) self (Variables|Default|GetGC_MiniGrenadeSpawned) (Ability|GameplayCue|MakeGameplayCueParameters 0.0)))

(event Custom|DoExplosionLatent
  ;[0.03429 0.3069 0.7899] Do delayed explosion
  (Delivery|DoExplosion))

(event EventOnExploded
  (params (HitActors :type Array%20of%20Actor%20Object%20References) (HitResults :type Array%20of%20Hit%20Result%20Structures))
  (bind _returnvalue (Transformation|GetActorLocation))
  ;[0.03429 0.3069 0.7899] Do delayed explosion
  (|CallParentFunction HitActors HitResults)
  (Networking|CanExecuteCosmeticEvents
    (:True
      (Fortnite|ShowSoundIndicatorLocal 0 (Transformation|GetActorLocation) 5e+03 "Gunshot")
      (if (or (Utilities|Enum|Equal(Enum) (Variables|Default|GetSurfaceType) "SurfaceType8") (Water|IsinWater (Variables|Water|GetWaterInteractionComponent)))
        (Effects|Components|ParticleSystem|SpawnFXAtLocation (Variables|Default|GetExplosion%20Asset_Water) _returnvalue (Transformation|GetActorRotation))
        (Audio|PlaySoundatLocation (Variables|Default|GetExplosionAudioWater) _returnvalue)
        (Audio|PlaySoundatLocation (Variables|Default|GetExplosionAudioWaterSplash) _returnvalue)
        (else
          (bind _returnvalue_2 (Effects|Components|ParticleSystem|SpawnFXAtLocation (Variables|Default|GetExplosion%20Asset) _returnvalue))
          (Audio|PlaySoundatLocation (Variables|Default|GetExplosionAudio) _returnvalue)
          (Effects|Components|ParticleSystem|SetBoolParameter _returnvalue_2 "Spawn Decal" (not (Variables|Default|GetbMainGrenadeStuckToMovingPlatform)))
          (Effects|Components|ParticleSystem|SetVectorParameter _returnvalue_2 "DecalRot" (Variables|Default|GetHitNormal)))))))

(event EventOnStop
  (params (Hit :type Hit%20Result%20Structure%20%28by%20ref%29))
  (bind (_bblockinghit _binitialoverlap _time _distance _location _impactpoint _normal _impactnormal) (Collision|BreakHitResult Hit))
  (Variables|Default|SetHitNormal _impactnormal))

(fn ConstructionScript)
